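; Audio controller for Papaya live events, handling show ducking, live volume, stage attenuation range and audio-reactive visuals.
; Mesh network events toggle show sound mixes on clients.
; Root-node commands set replicated variables, whose OnReps apply volume, distance scale and analysis settings.
; CQT spectrum and envelope follower code feeds material and Niagara parameter collections, but is marked as now handled by AudioAnalysisSubsystem, whose default submix registers at BeginPlay.
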(event Custom|Restart%20Spectral%20Analysis
  ;[0 0 0] Now handled by AudioAnalysisSubsystem
  ;;[1 1 1] CQT analysis
  (CallFunction|StopSpectralAnalysis)
  (CallFunction|StartSpectralAnalysis))

(event Custom|Stop%20Spectral%20Analysis
  (bind _analysissubmix (Variables|Default|GetAnalysisSubmix))
  ;[0 0 0] Now handled by AudioAnalysisSubsystem
  ;;[1 1 1] CQT analysis
  (Audio|Spectrum|RemoveSpectralAnalysisDelegate _analysissubmix (EventDispatchers|CreateEvent :!event Custom|On%20Spectral%20Analysis%20Update))
  (Audio|Analysis|StopSpectralAnalysis _analysissubmix))

(event Custom|Start%20Spectral%20Analysis
  (bind _analysissubmix (Variables|Default|GetAnalysisSubmix))
  ;[0 0 0] Now handled by AudioAnalysisSubsystem
  ;;[1 1 1] CQT analysis
  (Audio|Spectrum|AddSpectralAnalysisDelegate _analysissubmix (Variables|Spectrum|GetSpectrum_BandSettings) (delegate Custom|On%20Spectral%20Analysis%20Update) (Variables|Spectrum|GetSpectrum_UpdateRate) -4e+01 true (Variables|Default|GetAutoRangeEnabled) 0.1 6e+01)
  (Audio|Analysis|StartSpectralAnalysis _analysissubmix "DefaultSize" "ConstantQ" "Hamming" 0.0 "Decibel"))

(event Custom|On%20Spectral%20Analysis%20Update
  (params (Magnitude :type Array%20of%20Float%20%28single-precision%29s))
  (bind _spectrum_mpcname (Variables|Spectrum|GetSpectrum_MPCName))
  (bind _spectrum_mpcvalues (Variables|Spectrum|GetSpectrum_MPCValues))
  ;[0 0 0] Now handled by AudioAnalysisSubsystem
  ;;[1 1 1] EventAudio_Spectrum
  (Variables|Spectrum|SetSpectrum_RawCQT Magnitude)
  (if (Utilities|Enum|Equal(Enum) (Variables|Default|GetAnalysisType) "NewEnumerator1")
    (for (_array_element _array_index) (Variables|Spectrum|GetSpectrum_RawCQT)
      (CallFunction|SetColorValue _array_index (Variables|Spectrum|GetSpectrum_MPCValues) _array_element))
    (Rendering|Material|SetVectorParameterValue "/Game/Creative/Devices/Jukebox/Creative_Jukebox_MPC.Creative_Jukebox_MPC" _spectrum_mpcname _spectrum_mpcvalues :!class /Script/Engine.KismetMaterialLibrary)
    (Rendering|Material|SetVectorParameterValue "/Game/Packages/Fortress_SharedMaterials/GlobalMaterialParameters/FortniteMaterialParameters.FortniteMaterialParameters" "EventAudio_Spectrum" _spectrum_mpcvalues :!class /Script/Engine.KismetMaterialLibrary)
    (bind _returnvalue (Niagara|GetNiagaraParameterCollection "/Game/Athena/Sounds/Events/Papaya/FX/Papaya_Jukebox_NPC.Papaya_Jukebox_NPC"))
    (Niagara|SetVector4Parameter _returnvalue "Percussion" (Utilities|LinearColortoVector4 _spectrum_mpcvalues))
    (Default|CallSpectralAnalysisUpdate _spectrum_mpcname _spectrum_mpcvalues) ; Added so FX Stage Manager can access and handle Niagara Syncing
  ))

(event Custom|Envelope%20Follower%20Setup
  ;[0 0 0] Now handled by AudioAnalysisSubsystem
  ;;[1 1 1] Envelope follower on incoming audio from the analysis submix
  (Audio|EnvelopeFollowing|AddEnvelopeFollowerDelegate (Variables|Default|GetAnalysisSubmix) (delegate Custom|On%20Envelope%20Follower%20Update))
  (Audio|EnvelopeFollowing|StartEnvelopeFollowing (Variables|Default|GetAnalysisSubmix)))

(event Custom|On%20Envelope%20Follower%20Update
  (params (Envelope :type Array%20of%20Float%20%28single-precision%29s))
  (bind _output (Utilities|Array|Get(acopy) Envelope 1))
  (bind _output_2 (Utilities|Array|Get(acopy) Envelope))
  (bind _returnvalue (+ _output_2 _output))
  (bind _returnvalue_2 (Math|Float|Clamp(Float) _returnvalue 0.001))
  (bind _returnvalue_3 (Math|Float|Log _returnvalue_2 1e+01))
  (bind _returnvalue_4 (* _returnvalue_3 2e+01))
  (bind _returnvalue_5 (Math|Float|MapRangeClamped _returnvalue_4 -4e+01 -3.0 0.0 1.0))
  ;[0 0 0] Now handled by AudioAnalysisSubsystem
  ;;[1 1 1] Envelope follower on incoming audio from the analysis submix
  ;;; Is Array Valid
  (if (== (Utilities|Array|Length Envelope) 2)
    _)
  (Rendering|Material|SetScalarParameterValue "/Game/Creative/Devices/Jukebox/Creative_Jukebox_MPC.Creative_Jukebox_MPC" "AverageAmplitude" _returnvalue_5 :!class /Script/Engine.KismetMaterialLibrary)
  (Rendering|Material|SetScalarParameterValue "/Game/Packages/Fortress_SharedMaterials/GlobalMaterialParameters/FortniteMaterialParameters.FortniteMaterialParameters" "EventAudio_Amplitude" _returnvalue_5 :!class /Script/Engine.KismetMaterialLibrary)
  (bind _returnvalue_6 (Niagara|GetNiagaraParameterCollection "/Game/Athena/Sounds/Events/Papaya/FX/Papaya_Jukebox_NPC.Papaya_Jukebox_NPC"))
  (Niagara|SetFloatParameter _returnvalue_6 "AverageAmplitude" _returnvalue_5))

(event EventBeginPlay
  (bind _returnvalue (GameInstanceSubsystems|GetFortMeshNetworkEventsSubsystem))
  ;[1 1 1] Initialize
  (Utilities|FlowControl|Sequence
    (:then_0
      (Networking|CanExecuteCosmeticEvents
        (:True
          (AudioAnalysis|RegisterDefaultSubmix (WorldSubsystems|GetFortAudioAnalysisSubsystem))
          (CallFunction|PapayaHUD (Variables|Default|GetDebugVisibleOnBeginPlay))
          (Audio|PushSoundMixModifier (Variables|Default|GetLiveVolumeMixControlClass))
          (CallFunction|AmbientVolumeEventSetup))))
    (:then_1
      (bind _nodetype (Class|FortAsyncActionMeshNetworkReady|MeshNetworkReadyAsync)
        (:OnReady
          (switch Utilities|FlowControl|Switch|SwitchonEMeshNetworkNodeType _nodetype
            (:Client
              (EventDispatchers|AssignOnMeshNetworkEventBurst _returnvalue (EventDispatchers|CreateEvent :!event Custom|OnMeshNetEvent_Burst))
              (EventDispatchers|AssignOnMeshNetworkEventStart _returnvalue (EventDispatchers|CreateEvent :!event Custom|OnMeshNetEvent_Start))
              (EventDispatchers|AssignOnMeshNetworkEventEnd _returnvalue (EventDispatchers|CreateEvent :!event Custom|OnMeshNetEvent_End))
              (for _array_element (MeshNetworkEvents|GetActiveMeshNetworkEvents _returnvalue)
                (bind _array_element_eventname (Utilities|Struct|BreakMeshNetworkEventStateData _array_element))
                (switch name _array_element_eventname
                  (:PapayaShowOpen
                    (CallFunction|PapayaShow true))
                  (:PapayaShowClose
                    (CallFunction|PapayaShow false)))))))))))

(event Custom|Ambient%20Volume%20Event%20Setup)

(event Custom|On%20Ambient%20Entry%20Changed
  (params (AmbientName :type Name) (ActionType :type EAmbientAudioEntryActionType%20Enum :default "Added"))
  ;[1 1 1] Pushing/popping soundmixes when entering ambient volumes
  (CallFunction|PushSoundMixforAmbientVolume AmbientName ActionType))

(event Custom|setfrequencies
  (params (Frequencies :type String))
  ;[1 1 1] Spectral Analysis config commands
  (switch Utilities|FlowControl|Switch|SwitchonEMeshNetworkNodeType (MeshNetworking|GetMeshNetworkNodeType (Variables|Default|GetMeshNetwork))
    (:Root
      (Variables|Mesh|SetFrequencies Frequencies))))

(event Custom|setattacktimes
  (params (AttackTimes :type String))
  ;[1 1 1] Spectral Analysis config commands
  (switch Utilities|FlowControl|Switch|SwitchonEMeshNetworkNodeType (MeshNetworking|GetMeshNetworkNodeType (Variables|Default|GetMeshNetwork))
    (:Root
      (Variables|Mesh|SetAttackTime AttackTimes))))

(event Custom|setreleasetimes
  (params (ReleaseTimes :type String))
  ;[1 1 1] Spectral Analysis config commands
  (switch Utilities|FlowControl|Switch|SwitchonEMeshNetworkNodeType (MeshNetworking|GetMeshNetworkNodeType (Variables|Default|GetMeshNetwork))
    (:Root
      (Variables|Mesh|SetReleaseTime ReleaseTimes))))

(event Custom|setautorange
  (params (Enabled :type Boolean))
  ;[1 1 1] Spectral Analysis config commands
  (switch Utilities|FlowControl|Switch|SwitchonEMeshNetworkNodeType (MeshNetworking|GetMeshNetworkNodeType (Variables|Default|GetMeshNetwork))
    (:Root
      (Variables|Mesh|SetAutoRange Enabled))))

(event Custom|setvolume
  (params (Live%20Volume%20Mix%20Control%20Name :type Name) (Live%20Volume%20Mix%20Control%20Volume :type Float%20%28double-precision%29))
  ;[1 1 1] Commands to set attenuation scale and volume
  (switch Utilities|FlowControl|Switch|SwitchonEMeshNetworkNodeType (MeshNetworking|GetMeshNetworkNodeType (Variables|Default|GetMeshNetwork))
    (:Root
      (Variables|SetLiveVolumeMixControl_Name_9_328DB3524CF5A5410679B3B81813708E (Utilities|Struct|MakeSPapayaLiveVolumeMixControl Live%20Volume%20Mix%20Control%20Name Live%20Volume%20Mix%20Control%20Volume)))))

(event Custom|setaudiorange
  (params (Stage%20Audio%20Scale%20Name :type Name) (Stage%20Audio%20Scale%20Scale :type Float%20%28double-precision%29) (Stage%20Audio%20Scale%20Time :type Float%20%28double-precision%29))
  ;[1 1 1] Commands to set attenuation scale and volume
  (switch Utilities|FlowControl|Switch|SwitchonEMeshNetworkNodeType (MeshNetworking|GetMeshNetworkNodeType (Variables|Default|GetMeshNetwork))
    (:Root
      (Variables|SetStageAudioScale_StageName_9_D316F6EB4BC32836CE4389B1E25BE126 (Utilities|Struct|MakeSPapayaStageAudioScale Stage%20Audio%20Scale%20Name Stage%20Audio%20Scale%20Scale Stage%20Audio%20Scale%20Time)))))

(event Custom|audiotest
  ;[1 1 1] Fallback for audio
  (Audio|Components|Audio|Play (Variables|Default|GetAudio_Test) 0.0))

(event Custom|audiotesttoggle
  (params (Condition :type Boolean))
  (bind _audio_test (Variables|Default|GetAudio_Test))
  ;[1 1 1] Fallback for audio
  (if Condition
    (Audio|Components|Audio|Play _audio_test 0.0)
    (else
      (Audio|Components|Audio|Stop _audio_test))))

(event Custom|papayaHUD
  (params (Condition :type Boolean))
  ;[1 1 1] setup HUD for debug
  (if Condition
    (bind _debughud (Variables|Default|GetDebugHUD)
      (:then
        (if (Widget|IsVisible _debughud)
          _
          (else
            _)))
      (:else
        (bind _returnvalue (UserInterface|CreateWidget "/Game/Sounds/AudioAnalysis/DebugHUD/WB_AudioAnalysis_UI.WB_AudioAnalysis_UI_C"))
        (bind _output_get (Variables|Default|SetDebugHUD _returnvalue))
        (CallFunction|SetupDebugHUD)))
    (UserInterface|Viewport|AddtoViewport _output_get 0)
    (else
      (bind _debughud_2 (Variables|Default|GetDebugHUD)
        (:then
          (if (Widget|IsVisible _debughud_2)
            (Widget|RemovefromParent _debughud_2)))
        (:else)))))

(event Custom|OnMeshNetEvent_Burst
  (params (EventName :type Name) (NodeType :type EMeshNetworkNodeType%20Enum :default "Root") (EventParams :type String))
  ;[1 1 1] Mesh framework events
  (Development|PrintString (Utilities|String|Append "Event Burst [" (Utilities|String|ToString(Name) EventName) "] - Node Type [" (Utilities|String|EnumtoString NodeType) "]") true true "(R=0.000000,G=0.660000,B=1.000000,A=1.000000)" 3e+01))

(event Custom|PapayaShow
  (params (Open :type Boolean))
  (bind _showsoundmix (Variables|Default|GetShowSoundmix))
  ;[1 1 1] Extra security around audio ducking during shows
  (if (Math|Boolean|NotEqual(Boolean) Open (Variables|Default|GetbPapayaShowOpen))
    (bind _output_get (Variables|Default|SetbPapayaShowOpen Open))
    (if _output_get
      (Audio|PushSoundMixModifier _showsoundmix)
      (Audio|SetSoundMixClassOverride "/Game/Sounds/FortSoundClassesAndModes/Fort_Music_Emote_MixMod.Fort_Music_Emote_MixMod" "/Game/Sounds/FortSoundClassesAndModes/Fort_Live_InWorld_Papaya_Music.Fort_Live_InWorld_Papaya_Music")
      (Audio|SetSoundMixClassOverride "/Game/Sounds/FortSoundClassesAndModes/Fort_Music_Emote_3P_MixMod.Fort_Music_Emote_3P_MixMod" "/Game/Sounds/FortSoundClassesAndModes/Fort_Live_InWorld_Papaya_Music.Fort_Live_InWorld_Papaya_Music")
      (else
        (Audio|PopSoundMixModifier _showsoundmix)
        (Audio|SetSoundMixClassOverride "/Game/Sounds/FortSoundClassesAndModes/Fort_Music_Emote_MixMod.Fort_Music_Emote_MixMod" "/Game/Sounds/FortSoundClassesAndModes/Fort_Live_InWorld_Papaya_Music.Fort_Live_InWorld_Papaya_Music" 0.0)
        (Audio|SetSoundMixClassOverride "/Game/Sounds/FortSoundClassesAndModes/Fort_Music_Emote_3P_MixMod.Fort_Music_Emote_3P_MixMod" "/Game/Sounds/FortSoundClassesAndModes/Fort_Live_InWorld_Papaya_Music.Fort_Live_InWorld_Papaya_Music" 0.0)))))

(event Custom|OnMeshNetEvent_Start
  (params (EventName :type Name) (EventStartTime :type Date%20Time%20Structure) (NodeType :type EMeshNetworkNodeType%20Enum :default "Root") (EventParams :type String))
  ;[1 1 1] Mesh framework events
  (Development|PrintString (Utilities|String|Append "Event Started [" (Utilities|String|ToString(Name) EventName) "] - Start Time [" (Utilities|String|ToString(Text) (Utilities|Text|AsDateTime EventStartTime)) "] - Node Type [" (Utilities|String|EnumtoString NodeType) "]") true true "(R=0.000000,G=0.660000,B=1.000000,A=1.000000)" 3e+01)
  (switch Utilities|FlowControl|Switch|SwitchonEMeshNetworkNodeType NodeType
    (:Client
      (switch name EventName
        (:PapayaShowOpen
          (CallFunction|PapayaShow true))
        (:PapayaShowClose
          (CallFunction|PapayaShow false))))))

(event Custom|OnMeshNetEvent_End
  (params (EventName :type Name) (NodeType :type EMeshNetworkNodeType%20Enum :default "Root"))
  ;[1 1 1] Mesh framework events
  (Development|PrintString (Utilities|String|Append "Event Ended [" (Utilities|String|ToString(Name) EventName) "] - Node Type [" (Utilities|String|EnumtoString NodeType) "]") true true "(R=0.000000,G=0.660000,B=1.000000,A=1.000000)" 3e+01)
  (switch Utilities|FlowControl|Switch|SwitchonEMeshNetworkNodeType NodeType
    (:Client
      (switch name "None"))))

(fn GetColorValue
  (params (Index :type Integer) (Vec4 :type Linear%20Color%20Structure%20%28by%20ref%29))
  (bind (_r _g _b _a) (Utilities|Struct|BreakLinearColor Vec4))
  (switch int Index
    (:0
      (return _r))
    (:1
      (return _g))
    (:2
      (return _b))
    (:3
      (return _a))))

(fn OnRep_AttackTime
  (switch Utilities|FlowControl|Switch|SwitchonEMeshNetworkNodeType (MeshNetworking|GetMeshNetworkNodeType (Variables|Default|GetMeshNetwork))
    (:Client
      (for (_array_element _array_index) (CallFunction|ParseStringtoFloatArray (Variables|Mesh|GetAttackTime))
        (Utilities|Struct|SetmembersinSoundSubmixSpectralAnalysisBandSettings (Utilities|Array|Get(aref) (Variables|Spectrum|GetSpectrum_BandSettings) _array_index) (Math|Float|Truncate _array_element)))
      (CallFunction|RestartSpectralAnalysis))))

(fn OnRep_AutoRange
  (switch Utilities|FlowControl|Switch|SwitchonEMeshNetworkNodeType (MeshNetworking|GetMeshNetworkNodeType (Variables|Default|GetMeshNetwork))
    (:Client
      (CallFunction|RestartSpectralAnalysis))))

(fn OnRep_Frequencies
  (switch Utilities|FlowControl|Switch|SwitchonEMeshNetworkNodeType (MeshNetworking|GetMeshNetworkNodeType (Variables|Default|GetMeshNetwork))
    (:Client
      (for (_array_element _array_index) (CallFunction|ParseStringtoFloatArray (Variables|Mesh|GetFrequencies))
        (Utilities|Struct|SetmembersinSoundSubmixSpectralAnalysisBandSettings (Utilities|Array|Get(aref) (Variables|Spectrum|GetSpectrum_BandSettings) _array_index) _array_element))
      (CallFunction|StopSpectralAnalysis)
      (CallFunction|StartSpectralAnalysis))))

(fn OnRep_LiveVolumeMixControl
  (bind (_livevolumemixcontrol_name_9_328db3524cf5a5410679b3b81813708e _livevolumemixcontrol_volume_10_3d670b65470212ef7ba844b5c479ab6d) (Utilities|Struct|BreakSPapayaLiveVolumeMixControl (Variables|GetLiveVolumeMixControl_Name_9_328DB3524CF5A5410679B3B81813708E)))
  (Utilities|FlowControl|Sequence
    (:then_0
      (switch name _livevolumemixcontrol_name_9_328db3524cf5a5410679b3b81813708e
        (:Master
          (LocalVariables|SetSoundClassToApply "/Game/Sounds/FortSoundClassesAndModes/Fort_Live_InWorld.Fort_Live_InWorld"))
        (:Music
          (LocalVariables|SetSoundClassToApply "/Game/Sounds/FortSoundClassesAndModes/Fort_Live_InWorld_Papaya_Music.Fort_Live_InWorld_Papaya_Music"))
        (:Theater
          (LocalVariables|SetSoundClassToApply "/Game/Sounds/FortSoundClassesAndModes/Fort_Live_InWorld_Papaya_Theater.Fort_Live_InWorld_Papaya_Theater"))
        (:Loudspeakers
          (LocalVariables|SetSoundClassToApply "/Game/Sounds/FortSoundClassesAndModes/Fort_Live_InWorld_Papaya_Loudspeakers.Fort_Live_InWorld_Papaya_Loudspeakers"))))
    (:then_1
      (Audio|SetSoundMixClassOverride (Variables|Default|GetLiveVolumeMixControlClass) (LocalVariables|GetSoundClassToApply) _livevolumemixcontrol_volume_10_3d670b65470212ef7ba844b5c479ab6d))))

(fn OnRep_MusicVolume)

(fn OnRep_ReleaseTime
  (switch Utilities|FlowControl|Switch|SwitchonEMeshNetworkNodeType (MeshNetworking|GetMeshNetworkNodeType (Variables|Default|GetMeshNetwork))
    (:Client
      (for (_array_element _array_index) (CallFunction|ParseStringtoFloatArray (Variables|Mesh|GetReleaseTime))
        (Utilities|Struct|SetmembersinSoundSubmixSpectralAnalysisBandSettings (Utilities|Array|Get(aref) (Variables|Spectrum|GetSpectrum_BandSettings) _array_index) (Math|Float|Truncate _array_element)))
      (CallFunction|RestartSpectralAnalysis))))

(fn OnRep_StageAudioScale
  (bind (_stageaudioscale_stagename_9_d316f6eb4bc32836ce4389b1e25be126 _stageaudioscale_scale_2_328db3524cf5a5410679b3b81813708e _stageaudioscale_time_4_3d670b65470212ef7ba844b5c479ab6d) (Utilities|Struct|BreakSPapayaStageAudioScale (Variables|GetStageAudioScale_StageName_9_D316F6EB4BC32836CE4389B1E25BE126)))
  (Audio|SetSoundClassDistanceScale (select (Utilities|Name|Equal(Name) _stageaudioscale_stagename_9_d316f6eb4bc32836ce4389b1e25be126 "Music") (Variables|Default|GetMusicStageClass) (Variables|Default|GetTheaterStageClass)) (/ 1.0 (Math|Float|Max(Float) _stageaudioscale_scale_2_328db3524cf5a5410679b3b81813708e 0.01)) (Math|Float|Max(Float) _stageaudioscale_time_4_3d670b65470212ef7ba844b5c479ab6d 0.01)))

(fn OnRep_StageVolume)

(fn ParseStringToFloatArray
  (params (SourceString :type String))
  (bind _localfloats (LocalVariables|GetLocalFloats))
  (for _array_element (Utilities|String|ParseIntoArray SourceString ",")
    (Utilities|Array|Add _localfloats (Utilities|StripWhitespacetoFloat _array_element)))
  (return _localfloats))

(fn PushSoundMixforAmbientVolume
  (params (Key :type Name%20%28by%20ref%29) (Selection :type EAmbientAudioEntryActionType%20Enum :default "Added"))
  (bind _ambientvolumesoundmixmap (Variables|Default|GetAmbientVolumeSoundMixMap))
  (bind _value (Utilities|Map|Find _ambientvolumesoundmixmap Key))
  (switch Utilities|FlowControl|Switch|SwitchonEAmbientAudioEntryActionType Selection
    (:Added
      (Audio|PushSoundMixModifier _value))
    (:Removed
      (Audio|PopSoundMixModifier _value))))

(fn SetColorValue
  (params (Index :type Integer) (Vec4 :type Linear%20Color%20Structure%20%28by%20ref%29) (Value :type Float%20%28double-precision%29))
  (switch int Index
    (:0
      (Utilities|Struct|SetmembersinLinearColor Vec4 Value))
    (:1
      (Utilities|Struct|SetmembersinLinearColor Vec4 Value))
    (:2
      (Utilities|Struct|SetmembersinLinearColor Vec4 Value))
    (:3
      (Utilities|Struct|SetmembersinLinearColor Vec4 Value))))

(fn SetupDebugHUD)

(fn ConstructionScript)
