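# Sound cue for in-world music at the Papaya event stage, looping the Floss music pack wave through a single wave player.
# Stage attenuation, a reduced volume multiplier and the Papaya live music sound class shape its playback.
# A source bus send feeds the Papaya speaker system.
# Submix sends route it to a distant reverb with linear distance-based level and to a music audio-analysis submix.
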SoundCue /Game/Athena/Sounds/Events/Papaya/Papaya_AudioMusicCue
root=WavePlayer_0
nodes [{id=WavePlayer_0, properties {bLooping=true, soundWaveAssetPtr=/Game/Athena/Sounds/MusicPacks/Athena_MusicPack_Floss}}]
properties {
 attenuationSettings=&/Papaya_MusicStage_Event_Att
 maxDistance=50000
 soundClassObject=/Game/Sounds/FortSoundClassesAndModes/Fort_Live_InWorld_Papaya_Music
 volumeMultiplier=0.3
 busSends [{audioBus=null, minSendLevel=1, sendLevel=1, soundSourceBus=/Game/Athena/Sounds/Events/Papaya/SpeakerSystem/Papaya_SpeakerSourceBus}]
 soundSubmixSends [
  {
   maxSendDistance=10000
   minSendDistance=4000
   minSendLevel=0.1
   sendLevel=0
   sendLevelControlMethod=Linear
   soundSubmix=&Fortnite_Reverb_Distant
  }
  {minSendLevel=0, soundSubmix=&Fortnite_AudioAnalysis_Music}
 ]
}
